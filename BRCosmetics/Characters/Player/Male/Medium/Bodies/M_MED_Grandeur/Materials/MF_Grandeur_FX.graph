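# Material function for the Grandeur character skin that transitions between base and tinted body colors with a glowing bubbly edge, driven by actor altitude or debug/variant overrides.
# The transition value grows and moves a sphere mask in pre-skinned space, keeping it fixed to the animated body.
# Panning world-aligned bubble noise, warped by soft noise, erodes the mask.
# Curve-atlas rows shape the color blend and Fresnel-boosted emissive, while FX-mask channels select tint and logo regions.
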MaterialFunction /BRCosmetics/Characters/Player/Male/Medium/Bodies/M_MED_Grandeur/Materials/MF_Grandeur_FX
graph {nodes [
  n1 = PreSkinnedPosition()
  n2 = FunctionInput(Preview=ShaderStageSwitch(PixelShader=VertexInterpolator(VS=@n1).PS, VertexShader=@n1)) {bUsePreviewValueAsDefault=true, inputName=PreskinnedPos}
  n3 = NamedRerouteUsage()
  n4 = LinearInterpolate(
   A=ScalarParameter() {group=Grandeur, parameterName=StartRadius}
   B=ScalarParameter() {defaultValue=160, group=Grandeur, parameterName=EndRadius}
   Alpha=@n3
  )
  n5 = TextureObjectParameter() {group=Grandeur, parameterName=SL_TransitionNoise, texture=/Game/Characters/Player/Male/Medium/Bodies/M_MED_Viper/Textures/Glitch/T_SoftNoise}
  n6 = VectorParameter() {group=SuperLevelling_Mask, parameterName=SL_Backpack_PreskinnedPositionOffset, defaultValue=<rgba 0 0 0 1>}
  n7 = Add(A=@n2, B=@n6.RGB)
  n8 = VectorParameter() {group=Grandeur, parameterName=Gradient_DistortionNoise_PanDirectionAndSpeed, defaultValue=<rgba 2 2 2 1>}
  n9 = Multiply(A=Time(), B=@n8.RGB)
  n10 = FunctionInput(Preview=VertexInterpolator(VS=PreSkinnedNormal()).PS) {bUsePreviewValueAsDefault=true, inputName=Local Normal}
  n11 = ChannelMaskParameter(None=MaterialFunctionCall(
    TextureObject=@n5
    TextureSize=VectorParameter() {group=Grandeur, parameterName=Gradient_DistortionNoise_Tiling, defaultValue=<rgba 24 24 24 1>}.RGB
    WorldPosition=Add(A=@n7, B=@n9)
    World Space Normal=@n10
   ) {materialFunction=/Game/Weapons/WeaponSkins/Wraps/MaterialFunctions/WorldAlignedTexture}.XYZ Texture) {maskChannel=Green, defaultValue=<rgba 0 1 0 0>}
  n12 = Add(A=Floor(None=@n3), B=@n11.R)
  n13 = SphereMask(
   A=@n2
   B=MaterialFunctionCall(
    A=VectorParameter() {group=Grandeur, parameterName=SphereMaskPosition1, defaultValue=<rgba 12 0 148 1>}.RGB
    B=VectorParameter() {group=Grandeur, parameterName=SphereMaskPosition2, defaultValue=<rgba -4 0 176 1>}.RGB
    C=VectorParameter() {group=Grandeur, parameterName=SphereMaskPosition3, defaultValue=<rgba -44 0 100 1>}.RGB
    Alpha=NamedRerouteUsage()
   ) {materialFunction=&Lerp_3Color}.Result
   Radius=LinearInterpolate(A=Multiply(A=@n4) {constB=0.75}, B=@n4, Alpha=Clamp(None=@n12))
   Hardness=LinearInterpolate(
    A=ScalarParameter() {defaultValue=0.5, group=Grandeur, parameterName=hardness}
    B=ScalarParameter() {defaultValue=0.75, group=Grandeur, parameterName=hardness2}
    Alpha=@n3
   ) {constB=0.5}
  )
  n14 = TextureObjectParameter() {
   group=Grandeur
   parameterName=SL_TransitionNoise
   samplerType=SAMPLERTYPE_Masks
   texture=/BRCosmetics/Characters/Player/Male/Medium/Bodies/M_MED_Grandeur/Textures/T_Grandeur_Bubbles
  }
  n15 = Multiply(A=Time(), B=VectorParameter() {group=Cinematic, parameterName=Noise_PanDirectionAndSpeed, defaultValue=<rgba 0 0 -2 1>}.RGB)
  n16 = VectorParameter() {group=Grandeur, parameterName=SL_DistortionNoise_Tiling, defaultValue=<rgba 16 16 16 1>}
  n17 = Add(
   A=@n7
   B=Multiply(A=Time(), B=VectorParameter() {group=Grandeur, parameterName=CINE_DistortionNoise_PanDirectionAndSpeed, defaultValue=<rgba 4 4 2 1>}.RGB)
  )
  n18 = Subtract(A=Multiply(A=MaterialFunctionCall(TextureObject=@n5, TextureSize=@n16.RGB, WorldPosition=@n17, World Space Normal=@n10) {materialFunction=&WorldAlignedTexture}.XYZ Texture) {constB=2})
  n19 = ScalarParameter() {defaultValue=0.5, group=Grandeur, parameterName=DistortionStrength}
  n20 = MaterialFunctionCall(
   Export Float 4=StaticBool() {value=true}
   TextureObject=@n14
   TextureSize=VectorParameter() {group=Grandeur, parameterName=SL_TransitionNoise_Tiling, defaultValue=<rgba 90 90 90 1>}.RGB
   WorldPosition=Add(A=Add(A=@n15, B=@n7), B=Multiply(A=@n18, B=@n19))
   World Space Normal=@n10
  ) {materialFunction=&WorldAlignedTexture}
  n21 = ChannelMaskParameter(None=@n20.XYZ Texture) {group=SuperLevelling_Mask, maskChannel=Green, parameterName=SL_TransitionNoiseChannel, defaultValue=<rgba 0 1 0 0>}
  n22 = Floor(None=NamedRerouteUsage())
  n23 = Multiply(
   A=OneMinus(None=@n21.R)
   B=Multiply(A=ScalarParameter() {defaultValue=0.75, group=Grandeur, parameterName=NoiseStrength}, B=OneMinus(None=@n22))
  )
  n24 = Subtract(A=@n13, B=@n23)
  Ceil(None=@n24)
  CollectionParameter() {
   collection=&FortniteMaterialParameters
   parameterName=BR_Altitude_Zheight_Sealevel
  }
  n25 = Time()
  ConstantBiasScale(None=Sine(None=Multiply(A=@n25) {constB=0.5}))
  n26 = GetMaterialAttributes(None=FunctionInput() {inputName=Materials, inputType=FunctionInput_MaterialAttributes}) {attributeGetTypes [#1]}
  n27 = FunctionInput(Preview=TextureSample() {texture=/BRCosmetics/Characters/Player/Male/Medium/Bodies/M_MED_Grandeur/Textures/T_M_Grandeur_Body_FX}.RGBA) {inputName=FXMask, inputType=FunctionInput_Vector4}
  n28 = LinearInterpolate(
   A=@n26.BaseColor
   B=Multiply(A=VectorParameter() {group=Grandeur, parameterName=TintColor, defaultValue=<rgba 0.0625 0.0625 0.0625 1>}.RGB, B=@n26.BaseColor)
   Alpha=ChannelMaskParameter(None=@n27) {group=Grandeur, maskChannel=Green, parameterName=ColoswapChannel, defaultValue=<rgba 0 1 0 0>}.R
  )
  n29 = ChannelMaskParameter(None=@n27) {group=Grandeur, maskChannel=Blue, parameterName=LogoChannel, defaultValue=<rgba 0 0 1 0>}
  n30 = ScalarParameter() {group=Grandeur, parameterName=IsFaceAcc}
  n31 = LinearInterpolate(A=@n26.BaseColor, B=LinearInterpolate(A=@n26.BaseColor, B=@n28, Alpha=@n29.R), Alpha=@n30)
  n32 = LinearInterpolate(A=@n28, B=LinearInterpolate(A=@n28, B=@n26.BaseColor, Alpha=@n29.R), Alpha=@n30)
  n33 = CurveAtlasRowParameter(CurveTime=Clamp(None=@n24)) {
   atlas=&AthenaCurveAtlas01
   curve=/Game/Characters/Player/Male/Medium/Bodies/M_MED_Grandeur/Textures/GrandeurEdgeColorCurve
   defaultValue=306
   group=Grandeur
   parameterName=EdgeColorCurve
  }
  n34 = ScalarParameter() {defaultValue=0.5, group=Grandeur, parameterName=EdgeSamplePos}
  n35 = SphereMask(A=@n13, B=@n34, Radius=@n34, Hardness=ScalarParameter() {group=Grandeur, parameterName=EdgeSampleHardness})
  n36 = Multiply(A=@n23, B=ScalarParameter() {defaultValue=1, group=Grandeur, parameterName=BubbleNoise_Strength}) {constB=1.5}
  n37 = ChannelMaskParameter(None=@n20.XYZ Texture) {parameterName=Param_1}
  n38 = CurveAtlasRowParameter(CurveTime=@n37.R) {
   atlas=&AthenaCurveAtlas01
   curve=&GrandeurEdgeColorCurve
   defaultValue=306
   group=Grandeur
   parameterName=EdgeColorCurve
  }
  n39 = Multiply(A=Clamp(None=Subtract(A=@n35, B=@n36)), B=@n38)
  n40 = Fresnel(Normal=PixelNormalWS()) {exponent=2}
  n41 = LinearInterpolate(
   A=ScalarParameter() {defaultValue=1.5, group=Grandeur, parameterName=GlowStrengthmin}
   B=ScalarParameter() {defaultValue=3, group=Grandeur, parameterName=GlowStrengthmax}
   Alpha=Multiply(A=@n21.R, B=Power(Base=@n40))
  )
  FunctionOutput(None=SetMaterialAttributes!(
    MaterialAttributes=@n26.MaterialAttributes
    BaseColor=LinearInterpolate(A=@n31, B=@n32, Alpha=@n33.A)
    EmissiveColor=Multiply(A=@n39, B=@n41)
   ) {attributeSetTypes [#1, #2]})
  LinearInterpolate()
  LinearInterpolate(A=@n34, B=ScalarParameter() {defaultValue=0.5, group=Grandeur, parameterName=EdgeSamplePos02}, Alpha=@n11.R)
  LinearInterpolate(A=@n34, B=ScalarParameter() {defaultValue=0.5, group=Grandeur, parameterName=EdgeSamplePos2}, Alpha=NamedRerouteUsage())
  LinearInterpolate()
  Multiply(A=@n37.R, B=@n21.R)
  n42 = ChannelMaskParameter(None=ActorPositionWS()) {group=Grandeur, maskChannel=Blue, parameterName=ZLocation, defaultValue=<rgba 0 0 1 0>}
  n43 = CollectionParameter() {collection=&FortniteMaterialParameters, parameterName=BR_Altitude_ZMin}
  n44 = Add(A=@n43, B=ScalarParameter() {defaultValue=7140, parameterName=Min_Offset})
  n45 = CollectionParameter() {collection=&FortniteMaterialParameters, parameterName=BR_Altitude_ZMax}
  n46 = Add(A=@n45, B=ScalarParameter() {defaultValue=-12073, parameterName=Max_offset})
  n47 = Clamp(None=Divide(A=Subtract(A=@n42.R, B=@n44), B=Subtract(A=@n46, B=@n44)))
  n48 = ScalarParameter() {group=Grandeur, parameterName=DebugColorSwap}
  n49 = ScalarParameter() {group=Grandeur, parameterName=useDebugScalarForColorTint}
  NamedRerouteDeclaration(None=LinearInterpolate(
    A=LinearInterpolate(A=@n47, B=@n48, Alpha=@n49)
    B=ScalarParameter() {group=Grandeur, parameterName=VariantScalar}
    Alpha=ScalarParameter() {group=Grandeur, parameterName=ForceVariant}
   )) {name=TransitionControl, nodeColor=<rgba 0.9688 0 1 1>}
  OneMinus(None=@n47)
  ScalarParameter() {parameterName=DebugSCalar}
  ScalarParameter() {defaultValue=0.5, group=Grandeur, parameterName=EdgeSampleRadius}
  SphereMask(
   A=@n13
   B=ScalarParameter() {defaultValue=0.2, parameterName=ColorGradient_pos}
   Radius=ScalarParameter() {defaultValue=0.75, parameterName=ColorGradientRadius}
   Hardness=Constant()
  )
  TextureObjectParameter() {
   group=Grandeur
   parameterName=Gradient_TransitionNoise
   samplerType=SAMPLERTYPE_Masks
   texture=&T_Grandeur_Bubbles
  }
 ]}
