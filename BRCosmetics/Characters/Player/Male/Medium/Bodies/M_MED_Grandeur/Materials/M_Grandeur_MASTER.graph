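# Master character material for the Grandeur male medium body, layering skin-specific effects between the shared character PreFX and PostFX functions.
# PreFX output feeds MF_Grandeur_FX along with the SkinFX_Mask texture, then passes through optional cloth fuzz, clear coat and metal LUT stages via static switches.
# Static component masks select the cloth and clear coat channels from that mask.
# It uses material attributes with expression-driven shading models and a hero-skin subsurface profile.
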Material /BRCosmetics/Characters/Player/Male/Medium/Bodies/M_MED_Grandeur/Materials/M_Grandeur_MASTER
graph {
 properties {
  bUseMaterialAttributes=true
  bUsedWithMeshDeformer=true
  bUsedWithMorphTargets=true
  bUsedWithSkeletalMesh=true
  pixelDepthOffsetMode=PDOM_Legacy
  shadingModel=MSM_FromMaterialExpression
  subsurfaceProfile=&SS_HeroSkin_02
  usedShadingModels=Default Lit | Clear Coat | Subsurface Profile | Cloth
 }
 nodes [
  n1 = MaterialFunctionCall() {materialFunction=&BaseFN_CharacterFunction_PreFX}
  n2 = TextureSample(Tex=TextureObjectParameter() {group=Skin FX, parameterName=SkinFX_Mask})
  n3 = MaterialFunctionCall!(FXMask=@n2.RGBA, PreskinnedPos=@n1.Pre skinned Local Position, Local Normal, Materials=@n1.Pre skinned Local Position) {materialFunction=&/MF_Grandeur_FX}
  n4 = StaticSwitchParameter(
   True=MaterialFunctionCall!(
    Alpha=StaticComponentMaskParameter(None=@n2.RGBA) {defaultR=true, group=Skin FX, parameterName=Cloth Channel}
    Gmap=@n1.Gmap
    Material=@n3.Result
   ) {materialFunction=/Game/Weapons/WeaponSkins/MaterialFunction/ClothFuzz}.Result
   False=@n3.Result
  ) {group=Skin FX, parameterName=Use Cloth fuzz}
  n5 = StaticSwitchParameter(
   True=MaterialFunctionCall!(Alpha=StaticComponentMaskParameter(None=@n2.RGBA) {defaultR=true, group=Skin FX, parameterName=Clear Coat Channel}, In=@n4) {materialFunction=/Game/Weapons/WeaponSkins/MaterialFunction/Clearcoat}.Result
   False=@n4
  ) {group=Skin FX, parameterName=Use Clear Coat}
  MaterialFunctionCall(
   AO=@n1.AO
   Distance Blend=@n1.AO
   Fresnel=@n1.AO
   Gmap=@n1.AO
   Pre Skinned Local Position=@n1.AO
   Material=StaticSwitchParameter(
    True=MaterialFunctionCall(Mask=@n2.RGBA, Materials=@n5) {materialFunction=&MF_MetalsLUT}.Result
    False=@n5
   ) {group=Skin FX, parameterName=UseMetalLUT}
  ) {materialFunction=&BaseFN_CharacterFunction_PostFX}
 ]
}
