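# Animation Blueprint for the M_MED_Grandeur male body cosmetic, adding secondary motion to collar, zippers and shoelaces over a pose copied from the attached parent mesh.
# Head pitch-up, measured against spine_05 each update, drives the back collar's head rotation constraint and additive offsets.
# Weighted multi-bone constraints position chest and thigh zippers, then chained AnimDynamics springs swing zippers and shoelaces.
# The event graph also computes arm_up_r, which the AnimGraph never reads.
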AnimBlueprint /BRCosmetics/Characters/Player/Male/Medium/Bodies/M_MED_Grandeur/Meshes/M_MED_Grandeur_AnimBP
target_skeleton=&/M_MED_Grandeur_Skeleton
graphs [
 AnimGraph {nodes [
   n1 = Get Head Up() -> (Head Up:double)
   n2 = "Constraint - dyn_collar_bk (R)"(ComponentPose=Local To Component(LocalPose=Copy Pose From Mesh() {bUseAttachedParent=true}), Alpha=@n1, ConstraintWeights_0=1) {
    alphaBoolBlend {blendOption=Linear}
    boneToModify {boneName=dyn_collar_bk}
    constraintSetup [{transformType=Rotation, targetBone {boneName=head}}]
    constraintWeights [1]
   }
   n3 = "Transform (Modify) Bone - Bone: dyn_collar_bk"!(ComponentPose=@n2, Alpha=@n1, Rotation="0,-22,0.000") -> (Pose) {
    rotationMode=BMM_Additive
    rotationSpace=BCS_BoneSpace
    translationSpace=BCS_BoneSpace
    alphaBoolBlend {blendOption=Linear}
    boneToModify {boneName=dyn_collar_bk}
    rotation {yaw=-22}
   }
   n4 = "Transform (Modify) Bone - Bone: dyn_collar_bk"(ComponentPose=@n3, Alpha=@n1, Rotation="0,15,0") {
    rotationMode=BMM_Additive
    rotationSpace=BCS_BoneSpace
    translationSpace=BCS_BoneSpace
    alphaBoolBlend {blendOption=Linear}
    alphaScaleBiasClamp {bInterpResult=true}
    boneToModify {boneName=dyn_collar_bk}
    rotation {yaw=15}
   }
   n5 = "Constraint - dyn_chest_zip_r (PPPPP)"!(
    ComponentPose=@n4
    Alpha=1
    ConstraintWeights_0=0.0052
    ConstraintWeights_1:float=0.0109
    ConstraintWeights_2:float=0.0835
    ConstraintWeights_3:float=0.2675
    ConstraintWeights_4:float=0.6329
   ) -> (Pose) {
    alphaBoolBlend {blendOption=Linear}
    boneToModify {boneName=dyn_chest_zip_r}
    constraintSetup [
     {targetBone {boneName=clavicle_r}}
     {targetBone {boneName=neck_02}}
     {targetBone {boneName=spine_03}}
     {targetBone {boneName=spine_04}}
     {targetBone {boneName=spine_05}}
    ]
    constraintWeights [1,1,1,1,1]
   }
   n6 = "Constraint - dyn_chest_zip_l (PPPPP)"!(
    ComponentPose=@n5
    Alpha=1
    ConstraintWeights_0=0.0105
    ConstraintWeights_1:float=0.0837
    ConstraintWeights_2:float=0.2694
    ConstraintWeights_3:float=0.6313
    ConstraintWeights_4:float=0.005
   ) -> (Pose) {
    alphaBoolBlend {blendOption=Linear}
    boneToModify {boneName=dyn_chest_zip_l}
    constraintSetup [
     {targetBone {boneName=neck_02}}
     {targetBone {boneName=spine_03}}
     {targetBone {boneName=spine_04}}
     {targetBone {boneName=spine_05}}
     {targetBone {boneName=clavicle_l}}
    ]
    constraintWeights [1,1,1,1,1]
   }
   n7 = "Constraint - dyn_thigh_zip_top_r (PP)"!(ComponentPose=@n6, Alpha=1, ConstraintWeights_0=0.4436, ConstraintWeights_1:float=0.5564) -> (Pose) {
    alphaBoolBlend {blendOption=Linear}
    boneToModify {boneName=dyn_thigh_zip_top_r}
    constraintSetup [{targetBone {boneName=thigh_r}}, {targetBone {boneName=thigh_twist_01_r}}]
    constraintWeights [1,1]
   }
   n8 = "Constraint - dyn_thigh_zip_bot_r (PPP)"!(ComponentPose=@n7, Alpha=1, ConstraintWeights_0=0.6495, ConstraintWeights_1:float=0.0053, ConstraintWeights_2:float=0.3452) -> (Pose) {
    alphaBoolBlend {blendOption=Linear}
    boneToModify {boneName=dyn_thigh_zip_bot_r}
    constraintSetup [{targetBone {boneName=thigh_r}}, {targetBone {boneName=thigh_twist_01_r}}, {targetBone {boneName=calf_r}}]
    constraintWeights [1,1,1]
   }
   n9 = "Constraint - dyn_thigh_zip_top_l (PP)"!(ComponentPose=@n8, Alpha=1, ConstraintWeights_0=0.5564, ConstraintWeights_1:float=0.4436) -> (Pose) {
    alphaBoolBlend {blendOption=Linear}
    boneToModify {boneName=dyn_thigh_zip_top_l}
    constraintSetup [{targetBone {boneName=thigh_twist_01_l}}, {targetBone {boneName=thigh_l}}]
    constraintWeights [1,1]
   }
   n10 = "Constraint - dyn_thigh_zip_bot_l (PPP)"!(ComponentPose=@n9, Alpha=1, ConstraintWeights_0=0.3452, ConstraintWeights_1:float=0.0053, ConstraintWeights_2:float=0.6495) -> (Pose) {
    alphaBoolBlend {blendOption=Linear}
    boneToModify {boneName=dyn_thigh_zip_bot_l}
    constraintSetup [{targetBone {boneName=calf_l}}, {targetBone {boneName=thigh_twist_01_l}}, {targetBone {boneName=thigh_l}}]
    constraintWeights [1,1,1]
   }
   n11 = Anim Dynamics - Bone: dyn_chest_zip_l(ComponentPose=@n10, Alpha=1, ExternalForce="0,0,0") {
    angularDampingOverride=1
    angularSpringConstant=1500
    bAngularSpring=true
    bGravityOverrideInSimSpace=true
    bOverrideAngularDamping=true
    bUseGravityOverride=true
    gravityScale=0.35
    simulationSpace=RootRelative
    alphaBoolBlend {blendOption=Linear}
    boundBone {boneName=dyn_chest_zip_l}
    gravityOverride {z=-980}
    physicsBodyDefinitions [{
      sphereCollisionRadius=0
      boundBone {boneName=dyn_chest_zip_l}
      boxExtents=<10 20 20>
      constraintSetup {angularLimitsMax=<2 40 40>, angularLimitsMin=<-1.5 -40 -1>, angularTarget=<0 0 0>}
      localJointOffset=<4 0 0>
     }]
   }
   n12 = Anim Dynamics - Bone: dyn_chest_zip_r(ComponentPose=@n11, Alpha=1, ExternalForce="0,0,0") {
    angularDampingOverride=1
    angularSpringConstant=1500
    bAngularSpring=true
    bGravityOverrideInSimSpace=true
    bOverrideAngularDamping=true
    bUseGravityOverride=true
    gravityScale=0.35
    simulationSpace=RootRelative
    alphaBoolBlend {blendOption=Linear}
    boundBone {boneName=dyn_chest_zip_r}
    gravityOverride {z=-980}
    physicsBodyDefinitions [{
      sphereCollisionRadius=0
      boundBone {boneName=dyn_chest_zip_r}
      boxExtents=<10 20 20>
      constraintSetup {angularLimitsMax=<2 40 40>, angularLimitsMin=<-1.5 -40 -1>, angularTarget=<0 0 0>}
      localJointOffset=<4 0 0>
     }]
   }
   n13 = Anim Dynamics - Bone: dyn_thigh_zip_top_l(ComponentPose=@n12, Alpha=1, ExternalForce="0,0,0") {
    angularSpringConstant=1500
    bAngularSpring=true
    bGravityOverrideInSimSpace=true
    bUseGravityOverride=true
    lODThreshold=1
    numSolverIterationsPostUpdate=4
    numSolverIterationsPreUpdate=16
    alphaBoolBlend {blendOption=Linear}
    boundBone {boneName=dyn_thigh_zip_top_l}
    componentLinearAccScale=<0.25 1 3.6>
    componentLinearVelScale=<0.25 1 3.6>
    gravityOverride {z=-980}
    physicsBodyDefinitions [{
      sphereCollisionRadius=0
      boundBone {boneName=dyn_thigh_zip_top_l}
      boxExtents=<10 20 20>
      constraintSetup {angularLimitsMax=<2 20 60>, angularLimitsMin=<0 -5 -95>, angularTarget=<0 0 0>}
      localJointOffset=<5 0 0>
     }]
    sphericalLimits [{limitRadius=16, drivingBone {boneName=pelvis}, sphereLocalOffset=<-1 1 0>}]
   }
   n14 = Anim Dynamics - Bone: dyn_thigh_zip_top_r(ComponentPose=@n13, Alpha=1, ExternalForce="0,0,0") {
    angularDampingOverride=1
    angularSpringConstant=1500
    bAngularSpring=true
    bGravityOverrideInSimSpace=true
    bOverrideAngularDamping=true
    bUseGravityOverride=true
    gravityScale=0.35
    simulationSpace=RootRelative
    alphaBoolBlend {blendOption=Linear}
    boundBone {boneName=dyn_thigh_zip_top_r}
    componentLinearAccScale=<0.25 1 3.6>
    componentLinearVelScale=<0.25 1 3.6>
    gravityOverride {z=-980}
    physicsBodyDefinitions [{
      sphereCollisionRadius=0
      boundBone {boneName=dyn_thigh_zip_top_r}
      boxExtents=<10 20 20>
      constraintSetup {angularLimitsMax=<2 20 60>, angularLimitsMin=<0 -5 -95>, angularTarget=<0 0 0>}
      localJointOffset=<5 0 0>
     }]
   }
   n15 = Anim Dynamics - Bone: dyn_thigh_zip_bot_r(ComponentPose=@n14, Alpha=1, ExternalForce="0,0,0") {
    angularDampingOverride=1
    angularSpringConstant=1500
    bAngularSpring=true
    bGravityOverrideInSimSpace=true
    bOverrideAngularDamping=true
    bUseGravityOverride=true
    gravityScale=0.35
    simulationSpace=RootRelative
    alphaBoolBlend {blendOption=Linear}
    boundBone {boneName=dyn_thigh_zip_bot_r}
    componentLinearAccScale=<0.25 1 3.6>
    componentLinearVelScale=<0.25 1 3.6>
    gravityOverride {z=-980}
    physicsBodyDefinitions [{
      sphereCollisionRadius=0
      boundBone {boneName=dyn_thigh_zip_bot_r}
      boxExtents=<10 20 20>
      constraintSetup {angularLimitsMax=<3 20 60>, angularLimitsMin=<-2 -20 -95>, angularTarget=<0 0 0>}
      localJointOffset=<5 0 0>
     }]
   }
   n16 = Anim Dynamics - Bone: dyn_thigh_zip_bot_l(ComponentPose=@n15, Alpha=1, ExternalForce="0,0,0") {
    angularDampingOverride=1
    angularSpringConstant=1500
    bAngularSpring=true
    bGravityOverrideInSimSpace=true
    bOverrideAngularDamping=true
    bUseGravityOverride=true
    gravityScale=0.35
    simulationSpace=RootRelative
    alphaBoolBlend {blendOption=Linear}
    boundBone {boneName=dyn_thigh_zip_bot_l}
    componentLinearAccScale=<0.25 1 3.6>
    componentLinearVelScale=<0.25 1 3.6>
    gravityOverride {z=-980}
    physicsBodyDefinitions [{
      sphereCollisionRadius=0
      boundBone {boneName=dyn_thigh_zip_bot_l}
      boxExtents=<10 20 20>
      constraintSetup {angularLimitsMax=<3 20 60>, angularLimitsMin=<-2 -20 -95>, angularTarget=<0 0 0>}
      localJointOffset=<5 0 0>
     }]
   }
   n17 = Anim Dynamics - Bone: dyn_shoelace_l_01(ComponentPose=@n16, Alpha=1, ExternalForce="0,0,0") {
    alpha=0.9
    angularSpringConstant=1500
    bAngularSpring=true
    bGravityOverrideInSimSpace=true
    bUseGravityOverride=true
    numSolverIterationsPostUpdate=4
    numSolverIterationsPreUpdate=16
    simulationSpace=RootRelative
    alphaBoolBlend {blendOption=Linear}
    boundBone {boneName=dyn_shoelace_l_01}
    componentLinearAccScale {y=1, z=0.5}
    componentLinearVelScale {y=0.8, z=0.5}
    gravityOverride {z=-980}
    physicsBodyDefinitions [{
      sphereCollisionRadius=0
      boundBone {boneName=dyn_shoelace_l_01}
      boxExtents=<8 20 20>
      constraintSetup {angularLimitsMax=<0 15 30>, angularLimitsMin=<0 -5 0>, angularTarget=<0 0 0>}
      localJointOffset=<2 0 0>
     }]
   }
   n18 = Anim Dynamics - Bone: dyn_shoelace_r_01(ComponentPose=@n17, Alpha=1, ExternalForce="0,0,0") {
    alpha=0.9
    angularSpringConstant=1500
    bAngularSpring=true
    bGravityOverrideInSimSpace=true
    bUseGravityOverride=true
    numSolverIterationsPostUpdate=4
    numSolverIterationsPreUpdate=16
    simulationSpace=RootRelative
    alphaBoolBlend {blendOption=Linear}
    boundBone {boneName=dyn_shoelace_r_01}
    componentLinearAccScale {y=1, z=0.5}
    componentLinearVelScale {y=0.8, z=0.5}
    gravityOverride {z=-980}
    physicsBodyDefinitions [{
      sphereCollisionRadius=0
      boundBone {boneName=dyn_shoelace_r_01}
      boxExtents=<8 20 20>
      constraintSetup {angularLimitsMax=<0 15 30>, angularLimitsMin=<0 -5 0>, angularTarget=<0 0 0>}
      localJointOffset=<2 0 0>
     }]
   }
   Output Pose(Result=Component To Local(ComponentPose=@n18)) {layerGroup=None}
  ]}
 EventGraph {nodes [
   Event BlueprintUpdateAnimation()
   n1 = BreakTransform(InTransform=GetDeltaTransformFromRefPose(self=GetOwnerSkeletalMeshComponent(), BoneName=head, BaseName=spine_05))
   > Set Head Up(Head Up:double=MapRangeClamped(Value=BreakRotator(InRot=@n1.Rotation).Yaw, InRangeA=-3, InRangeB=-45)) -> (Output_Get:double)
   n2 = BreakTransform(InTransform=GetDeltaTransformFromRefPose(self=GetOwnerSkeletalMeshComponent(), BoneName=upperarm_r, BaseName=spine_05))
   > Set arm_up_r(arm_up_r:double=MapRangeClamped(Value=BreakRotator(InRot=@n2.Rotation).Yaw, InRangeA=60, InRangeB=100)) -> (Output_Get:double)
  ]}
]
variables [
 arm_up_r {type=number}
 Head Up {type=number}
]
